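# Translucent, two-sided lit material for spline-based smoke trails, apparently for the CozyTea Sprite Extractor machine.
# SplineThicken builds the trail width, with sine and cosine sway weighted along UV V and a camera offset added to world position offset.
# A panned soft smoke texture, masked along V and softened by depth and near-camera fades, drives opacity.
# UV1 offsets pan, sway timing and color-gradient exponent; an Angle Fade reroute is declared but unused.
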Material /Hermes/Props/CozyTea/CozyTea_SpriteExtractor_A/Materials/FX/M_SplineSmoke_MM_Machine
graph {
 properties {bAutomaticallySetUsageInEditor=false, blendMode=BLEND_Translucent, twoSided=true, usedShadingModels=Default Lit}
 nodes [
  n1 = ComponentMask(None=TextureCoordinate()) {g=true}
  n2 = Add(A=TextureCoordinate() {coordinateIndex=1}, B=ScalarParameter() {defaultValue=6, parameterName=Param_3})
  Add(
   A=Multiply(
    A=VectorParameter() {parameterName=Base Color, defaultValue=<rgba 0.6049 0 1 0.5>}.RGB
    B=ScalarParameter() {defaultValue=0.025, parameterName=Param_4}
   )
   B=Multiply(A=Power(Base=@n1, Exp=@n2), B=VectorParameter() {defaultValue=<rgba 0.25 0.0499 0 1>}.RGB)
  )
  n3 = MaterialFunctionCall(WidthBase=ScalarParameter(), WidthTip=ScalarParameter() {defaultValue=250, parameterName=Param_1}) {materialFunction=&SplineThicken}
  n4 = TextureCoordinate()
  n5 = OneMinus(None=Power(Base=ComponentMask(None=@n4) {g=true}))
  n6 = Add(A=Time(), B=ComponentMask(None=TextureCoordinate() {coordinateIndex=1}) {g=true})
  n7 = Sine(None=Multiply(A=@n6, B=ScalarParameter() {defaultValue=0.125, parameterName=Param_2}))
  n8 = Multiply(A=@n6, B=ScalarParameter() {defaultValue=0.06, parameterName=cosine time scale})
  n9 = ScalarParameter() {defaultValue=0.5, parameterName=cosine time offset}
  n10 = Multiply(A=Cosine(None=Add(A=@n8, B=@n9) {constB=0.5}), B=ScalarParameter() {defaultValue=1, parameterName=cosine amplitude})
  n11 = LinearInterpolate(
   A=VectorParameter() {parameterName=Param_2, defaultValue=<rgba 0 0 25 0>}.RGB
   B=VectorParameter() {parameterName=Param_1, defaultValue=<rgba 0 10 50 0>}.RGB
   Alpha=Saturate(None=Add(A=@n7, B=@n10))
  )
  Add(
   A=Add(A=@n3.WorldPositionOffset, B=Multiply(A=@n5, B=@n11))
   B=MaterialFunctionCall(Offset Amount=ScalarParameter() {defaultValue=5, parameterName=Param_6}, World Position=WorldPosition().XYZ) {materialFunction=/Engine/Functions/Engine_MaterialFunctions02/WorldPositionOffset/CameraOffset}.Camera Offset
  )
  n12 = VectorParameter() {parameterName=Param_3, defaultValue=<rgba 1 0.75 0 0>}
  n13 = TextureCoordinate() {coordinateIndex=1}
  n14 = Time()
  n15 = ScalarParameter() {defaultValue=0.125, parameterName=Smoke Pan Speed}
  n16 = Panner(
   Coordinate=Multiply(A=TextureCoordinate(), B=ComponentMask(None=@n12.RGB) {g=true, r=true})
   Time=Add(A=ComponentMask(None=@n13) {g=true}, B=Multiply(A=@n14, B=@n15))
  ) {speedY=1}
  n17 = Multiply(
   A=TextureSample(UVs=MaterialFunctionCall(XY=@n16) {materialFunction=&Swizzle}.YX) {samplerType=SAMPLERTYPE_Grayscale, texture=&T_Smoke_Trail_Soft}.RGB
   B=ComponentMask(None=TextureCoordinate()) {g=true}
  )
  n18 = ScalarParameter() {defaultValue=0.6, parameterName=Param_5}
  Multiply(
   A=DepthFade(Opacity=Multiply(A=@n17, B=@n18), FadeDistance=ScalarParameter() {defaultValue=30, parameterName=Depth Fade Dist})
   B=MaterialFunctionCall(Distance=ScalarParameter() {defaultValue=375, parameterName=Param_7}) {materialFunction=&Near_Camera_Fade}.Result
  )
  NamedRerouteDeclaration(None=@n3.Normal) {name=Normal, nodeColor=<rgba 1 0.9375 0 1>}
  n19 = DotProduct(A=CameraPositionWS(), B=NamedRerouteUsage())
  NamedRerouteDeclaration(None=OneMinus(None=Saturate(None=@n19))) {name=Angle Fade, nodeColor=<rgba 0.6172 0 1 1>}
  Time()
 ]
}
