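# Glass material for the CozyTea Sprite Extractor prop that fakes a glowing, swirling fluid inside a glossy Default Lit surface.
# Twisted, sine-warped UVs drive scrolling noise layers, bump-offset by their own noise for parallax depth, then combined into one fluid value.
# Smoothsteps map that value to a multi-color gradient, with Fresnel terms adding rim color and glass masking.
# Samples use explicit UV derivatives for mips, and inverse eye adaptation compensates emissive for exposure.
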Material /Hermes/Props/CozyTea/CozyTea_SpriteExtractor_A/Materials/M_CozyTea_SpriteExtractor_Glass
graph {
 properties {
  bAutomaticallySetUsageInEditor=false
  bUseMaterialAttributes=true
  previewMesh=/BattlePassS33_Quests/Items/QuestActors/Story/S33_Story_MM/CozyTea_SpriteExtractor_A/Meshes/SM_CozyTea_SpriteExtractorGlass_A
  usedShadingModels=Default Lit
 }
 nodes [
  n1 = TextureCoordinate()
  n2 = MaterialFunctionCall(Float2=@n1) {materialFunction=&BreakOutFloat2Components}
  n3 = NamedRerouteDeclaration(None=@n2.R) {name=UV_X, nodeColor=<rgba 1 0.9844 0 1>}
  n4 = ScalarParameter() {defaultValue=0.5, group=Sprite Extractor, parameterName=Sprite_L_TileX}
  n5 = ConstantBiasScale(None=Sine(None=@n2.R))
  n6 = ScalarParameter() {defaultValue=-0.15, group=Sprite Extractor, parameterName=Sprite_SineTwist}
  n7 = Add(A=@n2.G, B=Multiply(A=@n2.R, B=ScalarParameter() {defaultValue=0.3, group=Sprite Extractor, parameterName=Sprite_Twist}))
  n8 = NamedRerouteDeclaration(None=Add(A=Multiply(A=@n5, B=@n6), B=Frac(None=@n7))) {name=UV_Y, nodeColor=<rgba 0.6641 0 1 1>}
  n9 = ScalarParameter() {defaultValue=2, group=Sprite Extractor, parameterName=Sprite_L_TileY}
  n10 = NamedRerouteDeclaration(None=Multiply(A=Time(), B=ScalarParameter() {defaultValue=0.2, group=Sprite Extractor, parameterName=Time_Mult})) {name=Time, nodeColor=<rgba 0 0.6953 1 1>}
  n11 = ScalarParameter() {defaultValue=0.5, group=Sprite Extractor, parameterName=Sprite_L_ScrollX}
  n12 = ScalarParameter() {group=Sprite Extractor, parameterName=Sprite_L_ScrollY}
  n13 = Add(A=AppendVector(A=Multiply(A=@n3, B=@n4), B=Multiply(A=@n8, B=@n9)), B=AppendVector(A=Multiply(A=@n10, B=@n11), B=Multiply(A=@n10, B=@n12)))
  n14 = BumpOffset(Coordinate=@n13, Height=NamedRerouteUsage())
  n15 = NamedRerouteUsage()
  n16 = ScalarParameter() {defaultValue=2, group=Sprite Extractor, parameterName=Sprite_M_TileX}
  n17 = NamedRerouteUsage()
  n18 = ScalarParameter() {defaultValue=2, group=Sprite Extractor, parameterName=Sprite_M_TileY}
  n19 = NamedRerouteUsage()
  n20 = ScalarParameter() {defaultValue=0.7, group=Sprite Extractor, parameterName=Sprite_M_ScrollX}
  n21 = ScalarParameter() {defaultValue=-0.6, group=Sprite Extractor, parameterName=Sprite_M_ScrollY}
  n22 = Add(
   A=AppendVector(A=Multiply(A=@n15, B=@n16), B=Multiply(A=@n17, B=@n18))
   B=AppendVector(A=Multiply(A=@n19, B=@n20), B=Multiply(A=@n19, B=@n21))
  )
  n23 = BumpOffset(Coordinate=@n22, Height=NamedRerouteUsage())
  Add(A=@n14, B=@n23)
  n24 = TextureSample!(UVs=@n13, Tex, "DDX(UVs)"=NamedRerouteUsage(), "DDY(UVs)"=NamedRerouteUsage(), Apply View MipBias) {mipValueMode=TMVM_Derivative, samplerType=SAMPLERTYPE_Grayscale, texture=/Game/Global/Textures/Default/Noise/LowResBlurredNoiseNormalized}
  n25 = ScalarParameter() {defaultValue=-4, group=Sprite Extractor, parameterName=Bump Offset}
  n26 = BumpOffset(Coordinate=@n13, Height=Multiply(A=@n24.R, B=@n25) {constB=-4})
  n27 = TextureSample(UVs=@n22, "DDX(UVs)"=NamedRerouteUsage(), "DDY(UVs)"=NamedRerouteUsage()) {mipValueMode=TMVM_Derivative, samplerType=SAMPLERTYPE_Grayscale, texture=&LowResBlurredNoiseNormalized}
  n28 = BumpOffset(Coordinate=@n22, Height=Multiply(A=@n27.R, B=@n25) {constB=-4})
  n29 = TextureSample(
   UVs=Multiply(A=Add(A=@n26, B=@n28), B=Constant2Vector() {g=1, r=0.5}) {constB=3}
   "DDX(UVs)"=NamedRerouteUsage()
   "DDY(UVs)"=NamedRerouteUsage()
  ) {mipValueMode=TMVM_Derivative, samplerType=SAMPLERTYPE_LinearColor, texture=/Game/Global/Textures/Default/Noise/T_Noise01}
  n30 = Multiply(
   A=Add(A=NamedRerouteDeclaration(None=@n29.R) {name=Sprites_Small_M, nodeColor=<rgba 1 0 0.8672 1>}, B=Constant() {r=-0.5})
   B=Constant() {r=0.5}
  ) {constB=0.5}
  n31 = TextureSample(UVs=@n26, "DDX(UVs)"=NamedRerouteUsage(), "DDY(UVs)"=NamedRerouteUsage()) {mipValueMode=TMVM_Derivative, samplerType=SAMPLERTYPE_Grayscale, texture=&LowResBlurredNoiseNormalized}
  n32 = TextureSample(UVs=@n28, "DDX(UVs)"=NamedRerouteUsage(), "DDY(UVs)"=NamedRerouteUsage()) {mipValueMode=TMVM_Derivative, samplerType=SAMPLERTYPE_Grayscale, texture=&LowResBlurredNoiseNormalized}
  n33 = Add(A=@n32.R, B=Subtract(A=@n32.R, B=@n31.R))
  n34 = Divide(A=Add(A=Add(A=@n30, B=@n33))) {constB=3}
  n35 = NamedRerouteUsage()
  Add(A=@n34, B=Saturate(None=@n35))
  Constant3Vector() {constant=<rgba 0.7583 0.0529 2.5 1>}
  n36 = Multiply(A=NamedRerouteUsage(), B=ScalarParameter() {defaultValue=0.5, group=Sprite Extractor, parameterName=Sprite_Distortion_TileX})
  n37 = Multiply(A=NamedRerouteUsage(), B=ScalarParameter() {defaultValue=1, group=Sprite Extractor, parameterName=Sprite_Distortion_TileY})
  n38 = NamedRerouteUsage()
  n39 = Multiply(A=@n38, B=ScalarParameter() {defaultValue=1.1, group=Sprite Extractor, parameterName=Sprite_Distortion_ScrollX})
  n40 = Multiply(A=@n38, B=ScalarParameter() {defaultValue=0.5, group=Sprite Extractor, parameterName=Sprite_Distortion_ScrollY})
  n41 = TextureSample(UVs=Add(A=AppendVector(A=@n36, B=@n37), B=AppendVector(A=@n39, B=@n40)), "DDX(UVs)"=NamedRerouteUsage(), "DDY(UVs)"=NamedRerouteUsage()) {mipValueMode=TMVM_Derivative, samplerType=SAMPLERTYPE_Grayscale, texture=&LowResBlurredNoiseNormalized}
  n42 = Constant() {r=-5}
  NamedRerouteDeclaration(None=Add(A=Multiply(A=@n41.R, B=@n42), B=Constant() {r=-1}) {constB=0}) {name=Distortion Mask, nodeColor=<rgba 1 0.2812 0 1>}
  NamedRerouteDeclaration(None=DDX(Value=@n1)) {name=DDX, nodeColor=<rgba 1 0.4219 0 1>}
  NamedRerouteDeclaration(None=DDY(Value=@n1)) {name=DDY, nodeColor=<rgba 1 0.4219 0 1>}
  n43 = Add(A=LocalPosition().XYZ, B=Constant3Vector() {constant=<rgba -12.5 0 0 0>})
  n44 = Divide(
   A=ComponentMask(None=TransformPosition(None=@n43) {transformType=TRANSFORMPOSSOURCE_Camera}) {g=true, r=true}
   B=Constant2Vector() {g=80, r=100}
  ) {constB=1}
  n45 = Constant()
  NamedRerouteDeclaration(None=SmoothStep(Value=Distance(A=@n44, B=@n45)) {constMax=0, constMin=0.35}) {name=CoreBlob, nodeColor=<rgba 0.7109 0 1 1>}
  OneMinus(None=EyeAdaptation().EyeAdaptation)
  OneMinus(None=@n35)
  n46 = ScalarParameter() {defaultValue=0.1, group=Sprite Extractor, parameterName=Color3_Amount}
  OneMinus(None=@n46)
  ScalarParameter() {defaultValue=2, group=Sprite Extractor, parameterName=Sprite_Small_Tile}
  n47 = Add(A=NamedRerouteDeclaration(None=@n34) {name=Fluid Values, nodeColor=<rgba 0 1 0.9531 1>}, B=Multiply(A=NamedRerouteUsage()) {constB=-0.2})
  n48 = SmoothStep(
   Min=ScalarParameter() {defaultValue=0.02, group=Sprite Extractor, parameterName=Color1_Amount}
   Max=ScalarParameter() {defaultValue=0.35, group=Sprite Extractor, parameterName=Color2_Amount}
   Value=@n47
  ) {constMax=0.4, constMin=0.1}
  n49 = LinearInterpolate(
   A=LinearInterpolate(
    A=VectorParameter() {group=Sprite Extractor, parameterName=Sprite_Color1b, defaultValue=<rgba 2 0 0.1779 1>}.RGB
    B=VectorParameter() {group=Sprite Extractor, parameterName=Sprite_Color1c, defaultValue=<rgba 0.4427 0 2 1>}.RGB
    Alpha=NamedRerouteUsage()
   )
   B=VectorParameter() {group=Sprite Extractor, parameterName=Sprite_Color1a, defaultValue=<rgba 0.1417 0.8875 1.6 1>}.RGB
   Alpha=MaterialFunctionCall(In=Add(A=@n48) {constB=0.3}, Contrast=Constant() {r=1}) {materialFunction=&CheapContrast}.Result
  )
  n50 = VectorParameter() {group=Sprite Extractor, parameterName=Sprite_Color2, defaultValue=<rgba 0.0797 0.499 0.9844 1>}
  n51 = VectorParameter() {group=Sprite Extractor, parameterName=Sprite_Color3, defaultValue=<rgba 0.1371 0.0858 0.4705 1>}
  n52 = VectorParameter() {group=Sprite Extractor, parameterName=Sprite_Color4, defaultValue=<rgba 0.0225 0.0598 0.1823 1>}
  n53 = SmoothStep(Min=@n46, Max=ScalarParameter() {defaultValue=1, group=Sprite Extractor, parameterName=Color4_Amount}, Value=@n47) {constMin=0.4}
  n54 = Constant() {r=0.1}
  n55 = Constant() {r=0.45}
  n56 = NamedRerouteUsage()
  n57 = ScalarParameter() {defaultValue=0.5, group=Sprite Extractor, parameterName=fresnel}
  n58 = NamedRerouteDeclaration(None=DotProduct(A=CameraVectorWS(), B=VertexNormalWS())) {name=Fresnel, nodeColor=<rgba 0 1 0.3906 1>}
  n59 = NamedRerouteDeclaration(None=Step(Y=Constant() {r=0.25}, X=@n58) {constX=0.2, constY=0.2}) {name=Glass Mask, nodeColor=<rgba 1 0 0.5859 1>}
  n60 = SmoothStep(Min=Constant() {r=0.8}, Max=Constant(), Value=@n58) {constMax=0, constMin=1}
  n61 = Subtract(A=@n60, B=Multiply(A=Multiply(A=@n56, B=@n57), B=@n59))
  n62 = Max(
   A=LinearInterpolate(
    A=LinearInterpolate(A=@n49, B=@n50.RGB, Alpha=@n48)
    B=LinearInterpolate(A=@n51.RGB, B=@n52.RGB, Alpha=@n53)
    Alpha=SmoothStep(Min=@n54, Max=@n55, Value=@n47) {constMax=0.5, constMin=0.4}
   )
   B=Multiply(A=Add(A=@n60, B=@n61), B=VectorParameter() {group=Sprite Extractor, parameterName=Fresnel_Color, defaultValue=<rgba 0.2492 0 1 1>}.RGB)
  )
  n63 = LinearInterpolate(Alpha=NamedRerouteUsage()) {constA=0.15}
  SetMaterialAttributes!(
   MaterialAttributes
   Emissive Color=Multiply(A=Multiply(A=@n62, B=@n63), B=EyeAdaptationInverse().EyeAdaptationInverse)
   Opacity=LinearInterpolate(Alpha=NamedRerouteUsage()) {constA=0.7}
   Base Color=LinearInterpolate(A=VectorParameter() {group=Sprite Extractor, parameterName=Glass_Color, defaultValue=<rgba 0.195 1.0196 2 1>}.RGB, Alpha=Add(A=@n60, B=@n59)) {constB=0}
   Specular=Constant() {r=0.8}
   Roughness=Constant() {r=0.05}
  ) {attributeSetTypes [
    #1
    #2
    #3
    #4
    #5
   ]}
  Subtract(A=@n35, B=@n34)
  n64 = Constant() {r=-0.8}
  Subtract(A=@n34, B=Multiply(A=Add(A=@n35, B=@n64) {constB=-0.5}, B=Constant() {r=0.4}) {constB=0.2})
  TextureSample!(UVs, Tex, Apply View MipBias)
  TextureSample!(UVs=@n14, Tex, "DDX(UVs)"=NamedRerouteUsage(), "DDY(UVs)"=NamedRerouteUsage(), Apply View MipBias) {mipValueMode=TMVM_Derivative, samplerType=SAMPLERTYPE_Grayscale, texture=&LowResBlurredNoiseNormalized}
  TextureSample(UVs=@n23, "DDX(UVs)"=NamedRerouteUsage(), "DDY(UVs)"=NamedRerouteUsage()) {mipValueMode=TMVM_Derivative, samplerType=SAMPLERTYPE_Grayscale, texture=&LowResBlurredNoiseNormalized}
 ]
}
